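; NugZ variant of the Irwin prey NPC pawn, a chicken that reuses held-chicken bump-to-attack logic for repeated random pecking.
; AutoPeck attempts the player-bump ability, then reschedules itself by timer after a short random delay; nothing here starts that loop.
; IsValidAutoFireTarget accepts any actor.
; Bump assign/unassign and BeginPlay only call their parents, and the customization-loaded handler is empty despite a note about a material workaround.
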(event EventAssignPlayerBumpEvent
  ;[1 1 1] Hook into held chicken bump-to-attack logic for our looping random attack thing.
  (|CallParentFunction))

(event EventUnassignPlayerBumpEvent
  ;[1 1 1] Hook into held chicken bump-to-attack logic for our looping random attack thing.
  (|CallParentFunction))

(event Custom|AutoPeck
  ;[1 1 1] Hook into held chicken bump-to-attack logic for our looping random attack thing.
  (CallFunction|AttemptTriggerAbilityAfterPlayerBumpEvent)
  (Utilities|Time|SetTimerbyFunctionName 0 "AutoPeck" (Math|Random|RandomFloatinRange 1.0 2.5)))

(event EventBeginPlay
  ;[1 1 1] Hack Mountain - Tried to apply material through Customization, but it doesn't seem to work.
  (|CallParentFunction))

(event Custom|OnCustomizationsLoadedEvent_Event
  (params (ThisPawn :type Fort%20Pawn%20Object%20Reference)))

(fn IsValidAutoFireTarget
  (params (Attacker :type Actor%20Object%20Reference))
  (return true))

(fn ConstructionScript
  (|CallParentFunction))
